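# M_3dAirGlow is an additive unlit material for glow meshes and mesh particles, tinted by time of day.
# Fresnel shapes the color, with depth fade skipped on low quality, before an emissive adjustment function.
# Camera-to-actor distance drives a fade, multiplied with time-of-day alpha and a sine-based breathing term.
# A ceiled copy of that fade switches vertex offset between collapse to the object position and a normal push.
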Material /WildEstateTech/FX/FarmGlow/M_3dAirGlow
graph {
 properties {
  bUsedWithMeshParticles=true
  bUsedWithNanite=true
  bUsedWithNiagaraMeshParticles=true
  blendMode=BLEND_Additive
  pixelDepthOffsetMode=PDOM_Legacy
  shadingModel=MSM_Unlit
  usedShadingModels=Unlit
 }
 nodes [
  n1 = Length(None=Subtract(A=CameraPositionWS(), B=ActorPositionWS())) {is=class MaterialExpressionLength}
  n2 = ScalarParameter() {parameterName="Fade Offset (S)"}
  n3 = Saturate(None=Divide(A=Subtract(A=@n1, B=@n2), B=ScalarParameter() {parameterName="Fade Length (S)"}))
  LinearInterpolate(
   A=Subtract(A=ObjectPositionWS(), B=WorldPosition().XYZ)
   B=Multiply(A=VertexNormalWS(), B=ScalarParameter() {parameterName=Vert Push})
   Alpha=Ceil(None=@n3)
  ) {constB=0}
  n4 = VectorParameter() {parameterName=Color alpha morning}
  n5 = VectorParameter() {parameterName=Color alpha day}
  n6 = VectorParameter() {parameterName=Color alpha evening}
  n7 = VectorParameter() {parameterName=Color alpha night}
  n8 = MaterialFunctionCall(
   Brightness=ScalarParameter() {defaultValue=1, parameterName=Fresnel Brightness}
   Exponent=ScalarParameter() {defaultValue=10, parameterName=Fresnel Exponent}
  ) {materialFunction=&MF_Fresnel_FX}
  n9 = Multiply(
   A=MaterialFunctionCall(Color_Morning=@n4.RGB, Color_Day=@n5.RGB, Color_Evening=@n6.RGB, Color_Night=@n7.RGB) {materialFunction=&Color_Time_of_Day}.Result
   B=LinearInterpolate(A=@n8.Fresnel, B=@n8.Fresnel, Alpha=ScalarParameter() {parameterName=use Double sided Fres})
  )
  n10 = DepthFade(Opacity=@n9, FadeDistance=ScalarParameter() {defaultValue=500, parameterName=Depth Fade Distance})
  MaterialFunctionCall(Emissive=QualitySwitch(Default=@n10, Low=@n9, High=@n10, Medium=@n10)) {materialFunction=&MF_FX_EmissiveAdjustments}
  Multiply()
  n11 = MaterialFunctionCall(Value_Morning=@n4.A, Value_Day=@n5.A, Value_Evening=@n6.A, Value_Night=@n7.A) {materialFunction=&Scalar_Time_of_Day}
  n12 = VectorParameter() {parameterName=Brealth options, channelNames {a="", b=Speed, g=High, r=low}, defaultValue=<rgba 0.5 1 1 0>}
  n13 = LinearInterpolate(
   A=@n12.low
   B=@n12.low
   Alpha=MaterialFunctionCall(Speed=Divide(B=@n12.Speed)) {materialFunction=&DebugTimeSine}.Result
  )
  Multiply(A=Multiply(A=Multiply(A=@n11.Result, B=@n13), B=@n3), B=ScalarParameter() {defaultValue=1, parameterName=Phase Opacity Mod})
  Reroute(None=@n1)
  ScalarParameter() {parameterName=Master Opac}
 ]
}
